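# Translucent, unlit, two-sided particle sprite material for a glittering wall effect.
# Two copies of a tornado tile texture pan in different directions at different tiling, offset by a dynamic parameter, and multiply into a boosted shimmer mask.
# That mask modulates the wall glitter texture, particle color and particle alpha.
# The result feeds a shared emissive adjustment function.
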Material /Game/Effects/Fort_Effects/Materials/Walls/M_Wall_Glitter
graph {
 properties {
  bEnableResponsiveAA=true
  bUsedWithNanite=true
  bUsedWithParticleSprites=true
  blendMode=BLEND_Translucent
  pixelDepthOffsetMode=PDOM_Legacy
  shadingModel=MSM_Unlit
  twoSided=true
  usedShadingModels=Unlit
 }
 nodes [
  n1 = ParticleColor()
  n2 = TextureSample() {texture=/Game/Effects/Fort_Effects/Textures/T_Wall_Glitter}
  n3 = DynamicParameter() {defaultValue=<rgba 0 0 0 1>}
  n4 = AppendVector(A=@n3.Param1, B=@n3.Param1)
  n5 = Panner(Coordinate=Add(A=TextureCoordinate(), B=@n4)) {speedX=0.5, speedY=-0.8}
  n6 = Panner(Coordinate=Add(A=TextureCoordinate() {uTiling=4, vTiling=4}, B=@n4)) {speedX=-0.2, speedY=0.4}
  n7 = Multiply(
   A=Multiply(
    A=TextureSample(UVs=@n5) {texture=/Game/Global/Textures/Default/Tileable_Patterns/T_Tornado_Tile}.G
    B=TextureSample(UVs=@n6) {texture=&T_Tornado_Tile}.G
   )
   B=Constant() {r=5}
  )
  n8 = Multiply(A=@n1.A, B=Multiply(A=@n1.RGB, B=Multiply(A=@n2.RGB, B=@n7)))
  MaterialFunctionCall(Emissive=@n8) {materialFunction=&MF_FX_EmissiveAdjustments}
  Multiply()
  Multiply(A=@n8, B=Constant() {r=0.8})
 ]
}
